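# Sound cue for the lock sound of the Electric Vault door at a POI.
# A single wave player plays the ElectricDoorVault_Lock wave with a slightly boosted volume.
# It uses the shared Fort door attenuation and sound class, and overrides concurrency to cap simultaneous instances with a short retrigger time.
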SoundCue /Hera/Audio/POI/ElectricVault/SC_ElectricDoorVault_Lock
root=WavePlayer_0
nodes [{id=WavePlayer_0, properties {soundWaveAssetPtr=&/ElectricDoorVault_Lock}}]
properties {
 attenuationSettings=/Game/Sounds/Attenuation/Fort_Doors_Att
 bOverrideConcurrency=true
 maxDistance=2400
 soundClassObject=&Fort_Doors
 volumeMultiplier=1.25
 concurrencyOverrides {maxCount=3, retriggerTime=0.1}
}
