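# Sound cue that plays the unlock sound for the electric vault door in the ElectricVault point of interest.
# A single wave player feeds a fixed short delay, then a mixer whose other input is left empty.
# The cue is pitched slightly down, uses the Fort door sound class and attenuation, and overrides concurrency to cap simultaneous plays and set a retrigger time.
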SoundCue /Hera/Audio/POI/ElectricVault/SC_ElectricDoorVault_Unlock
root=Mixer_0
nodes [
 {id=Delay_0, children [WavePlayer_0], properties {delayMax=0.15, delayMin=0.15}}
 {id=Mixer_0, children [null,Delay_0], properties {inputVolume [0.25,1]}}
 {id=WavePlayer_0, properties {soundWaveAssetPtr=&/ElectricDoorVault_Unlock}}
]
properties {
 attenuationSettings=/Game/Sounds/Attenuation/Fort_Doors_Att
 bOverrideConcurrency=true
 maxDistance=2400
 pitchMultiplier=0.9
 soundClassObject=&Fort_Doors
 volumeMultiplier=1
 concurrencyOverrides {maxCount=3, retriggerTime=0.1}
}
