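# Looping lockdown alarm sound cue for the Electric Vault POI door, named as the electric door alarm in Hera audio.
# A fixed short delay node feeds a looping wave player that plays a heist POI lockdown alarm wave.
# It uses a large spot attenuation without occlusion, the gameplay events sound class, reduced volume, and a concurrency override capping simultaneous instances.
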SoundCue /Hera/Audio/POI/ElectricVault/SC_ElectricDoor_Alarm_Loop
root=Delay_0
nodes [
 {id=Delay_0, children [WavePlayer_1], properties {delayMax=0.3, delayMin=0.3}}
 {id=WavePlayer_1, properties {bLooping=true, soundWaveAssetPtr=/Game/Sounds/Asteria/POI/HeistPOIs/wav/Alarm_Lockdown_LP_02}}
]
properties {
 attenuationSettings=/Hera/Audio/Attenuations/ATT_Hera_Spot_Large_NoOcclusion
 maxDistance=2300
 soundClassObject=&Fort_Events_Gameplay
 volumeMultiplier=0.5
 concurrencyOverrides {bVolumeScaleCanRelease=true, maxCount=3}
 concurrencySet [null]
}
