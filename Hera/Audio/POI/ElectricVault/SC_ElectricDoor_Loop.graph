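# Looping electricity ambience cue for the electric door at the Electric Vault POI, built from DamageRails electricity loops and routed to the Fort_Traps sound class.
# A random pick of two loops, one slightly quieter, mixes with a quieter, pitched-down third loop.
# The mix passes through an overridden attenuation node and a root modulator for slight pitch variation.
# Concurrency override caps simultaneous instances, stopping the quietest.
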SoundCue /Hera/Audio/POI/ElectricVault/SC_ElectricDoor_Loop
root=Modulator_0
nodes [
 {
  id=Attenuation_4
  children [Mixer_0]
  properties {bOverrideAttenuation=true, attenuationOverrides {falloffDistance=600, attenuationShapeExtents {x=300}}}
 }
 {id=Mixer_0, children [Random_0,Modulator_1], properties {inputVolume [1,0.5]}}
 {id=Modulator_0, children [Attenuation_4], properties {pitchMax=1.1, pitchMin=0.9}}
 {id=Modulator_1, children [WavePlayer_2], properties {pitchMax=0.9, pitchMin=0.9}}
 {id=Modulator_2, children [WavePlayer_0], properties {volumeMax=0.9, volumeMin=0.9}}
 {id=Random_0, children [Modulator_2,WavePlayer_1], properties {weights [1,1]}}
 {id=WavePlayer_0, properties {bLooping=true, soundWaveAssetPtr=/Game/Sounds/Creative/Toys/DamageRails/DamageRails_Loop_Electricity01}}
 {id=WavePlayer_1, properties {bLooping=true, soundWaveAssetPtr=/Game/Sounds/Creative/Toys/DamageRails/DamageRails_Loop_Electricity02}}
 {id=WavePlayer_2, properties {bLooping=true, soundWaveAssetPtr=/Game/Sounds/Creative/Toys/DamageRails/DamageRails_Loop_Electricity03}}
]
properties {
 bOverrideConcurrency=true
 maxDistance=900
 soundClassObject=&Fort_Traps
 volumeMultiplier=0.5
 attenuationOverrides {falloffDistance=1600}
 concurrencyOverrides {maxCount=3, resolutionRule=StopQuietest}
}
