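# Sound cue for the grid light activation sound in the Hera ElectricVault POI.
# A single wave player feeds GridLight_Activate straight to the output at boosted volume, with no randomization or mixing.
# It uses the large spot, no-occlusion Hera attenuation and the Athena ambient props sound class.
# A borrowed Creative trigger-receiver concurrency, named for allowing only one instance, governs overlapping plays.
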SoundCue /Hera/Audio/POI/ElectricVault/SC_GridLight_Activate
root=WavePlayer_1
nodes [{id=WavePlayer_1, properties {soundWaveAssetPtr=&/GridLight_Activate}}]
properties {
 attenuationSettings=/Hera/Audio/Attenuations/ATT_Hera_Spot_Large_NoOcclusion
 maxDistance=2300
 soundClassObject=&Fort_Ambient_Props_Athena
 volumeMultiplier=2
 concurrencySet [/Game/Sounds/Creative/Concurrency/TriggerReciever_OnlyOne_Conc]
}
