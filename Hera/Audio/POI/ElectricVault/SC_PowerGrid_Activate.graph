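# Sound cue for the power grid activation sound at the Electric Vault point of interest.
# It plays a single PowerGrid_Activate wave through one wave player node with no randomization or mixing.
# It uses the Hera medium spot attenuation, the Athena ambient props sound class and a boosted volume multiplier.
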SoundCue /Hera/Audio/POI/ElectricVault/SC_PowerGrid_Activate
root=WavePlayer_0
nodes [{id=WavePlayer_0, properties {soundWaveAssetPtr=&/PowerGrid_Activate}}]
properties {
 attenuationSettings=/Hera/Audio/Attenuations/ATT_Hera_Spot_Medium
 maxDistance=1500
 soundClassObject=&Fort_Ambient_Props_Athena
 volumeMultiplier=2
}
