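# Sound cue for a security generator stopping at the Electric Vault POI, played as an ambient prop sound.
# A mixer layers a louder vault unlock sound with a quiet, evenly weighted random pick of three electric impact sounds, and its third input is empty.
# The mix feeds a root modulator and uses large spot attenuation without occlusion, with a concurrency override that prevents new instances at the limit.
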SoundCue /Hera/Audio/POI/ElectricVault/SC_Security_Generator_Stop
root=Modulator_1
nodes [
 {id=Mixer_1, children [WavePlayer_3,Random_0,null], properties {inputVolume [1.25,0.25,1]}}
 {id=Modulator_1, children [Mixer_1]}
 {id=Random_0, children [WavePlayer_7,WavePlayer_8,WavePlayer_9], properties {weights [1,1,1]}}
 {id=WavePlayer_3, properties {soundWaveAssetPtr=&/Vault_Unlock}}
 {id=WavePlayer_7, properties {soundWaveAssetPtr=/Game/Sounds/Fort_GamePlay_Sounds/Elemental_Layers/enemy_electric_medium_impact_03}}
 {id=WavePlayer_8, properties {soundWaveAssetPtr=/Game/Sounds/Fort_GamePlay_Sounds/Elemental_Layers/enemy_electric_medium_impact_02}}
 {id=WavePlayer_9, properties {soundWaveAssetPtr=/Game/Sounds/Fort_GamePlay_Sounds/Elemental_Layers/enemy_electric_medium_impact_01}}
]
properties {
 attenuationSettings=/Hera/Audio/Attenuations/ATT_Hera_Spot_Large_NoOcclusion
 bOverrideConcurrency=true
 maxDistance=2300
 soundClassObject=&Fort_Ambient_Props_Athena
 volumeMultiplier=1.25
 concurrencyOverrides {resolutionRule=PreventNew}
}
